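# Niagara system for the Singularity vehicle drift trail cosmetic, combining two fading ribbons, wheel-area sprites and looping wheel-wrap meshes.
# Sprites and meshes attach to skeletal bones by partition index, with three sprite spawn groups tagged for separate renderers.
# Each emitter takes per-variant paint colors and masks particles per tire via FN_DisableTireFX.
# Ribbons and sprites inherit source movement through decaying curves, while VC_WheelWrapScale sizes the meshes.
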NiagaraSystem /VehicleCosmetics/DriftTrail/DriftTrail_Singularity/FX/NS_DriftTrail_Singularity
dependencies {
 usedDataInterfaces [&NiagaraDataInterfaceCurve, &NiagaraDataInterfaceSkeletalMesh]
 usedDynamicInputs [/Niagara/DynamicInputs/LinearColor/MakeLinearColorFromHSV, &RandomRangeFloat]
 usedModules [
  &self:SystemSpawnScript.SystemScriptSource.SystemScriptGraph.NiagaraNodeAssignment_0.SetVariables_#1
  &FN_UnpackVehicleFlags
  /VehicleCosmetics/FX/Niagara/Modules/VC_SetVehicleSpeedAttribs
  &SystemState
 ]
}
emitters {
 NE_VC_Ribbon001 {
  emitterUpdateScript [
   &EmitterState(Life Cycle Mode=::NewEnumerator1, Loop Duration=1, UseLoopDelay=false)
   &VC_RateQualityHandler(Epic Spawn Rate=90, High Spawn Rate=60, Low Spawn Rate=30, Medium Spawn Rate=45, Use Custom Rate?=true)
   /SetVariables(Emitter.TextureDesaturation:NiagaraFloat=&MakeCustomFloatFromBool(Bool=$User.bIsPainted, True Float=1, False Float=0)) {isSetParameters=true}
   &VC_SpawnRibbon_v1(Number of Ribbons=2, SpawnRate=$Emitter.Spawn Rate)
  ]
  particleSpawnScript [
   &VC_PaintedVariantsOverride(
    Black=<rgba 0.032 0.032 0.032 7.5>
    BurntSienna=<rgba 1 0.5833 0.5 10>
    Cobalt=<rgba 0.4375 0.5031 1.75 7.5>
    Crimson=<rgba 1.25 0.3646 0.1875 6>
    DefaultColor=<rgba 0.7 0 2.4 25>
    ForestGreen=<rgba 0.2871 1.3 0.195 20>
    Grey=<rgba 0.6562 0.6562 0.7 4>
    Lime=<rgba 0.9333 1.75 0.35 3>
    Orange=<rgba 1.2 0.7 0 5>
    Pink=<rgba 1.0125 0.525 1.5 10>
    Purple=<rgba 0.1667 0 0.5 10>
    Saffron=<rgba 1.5 1.5 0.075 4>
    SkyBlue=<rgba 0.3125 1.25 1.0937 5>
    TitaniumWhite=<rgba 2 2 2 2>
   )
   &VC_InitializeRibbon_v1(
    Color=$Particles.Color
    Custom Offset=true
    Enable Skeletal Position=true
    Lifetime=0.7
    Local Vector Offset=&AddVectorToPosition(Position=$User.offset, Vector=<-7 0 -30>)
    Offset Amount=1
    Ribbon Width=65
   )
   &AddVelocity(Velocity=<-1500 0 0>) {module=AddVelocity001}
  ]
  particleUpdateScript [
   &FN_DisableTireFX(Mask Index Attribute:NiagaraInt32=$Particles.SpawnGroup, Mask Index FL=2, Mask Index FR=3, Mask Index RL=0, Mask Index RR=1)
   &ParticleState()
   &InheritSourceMovement(
    Applied Position Scale=&Multiply_VectorByFloat(
     Vector=<1 1 1>
     Float=&FloatFromCurve(FloatCurve=DataInterface() {propertyValues {Curve {keys [
          {interpMode=RCIM_Cubic, leaveTangent=-0.1311, tangentMode=RCTM_Break, time=0, value=1}
          {arriveTangent=-0.0803, interpMode=RCIM_Cubic, tangentMode=RCTM_Break, time=1, value=0.9}
         ]}}})
    )
    ApplyForce=false
    ApplyPosition=true
    ApplyPositionFalloff=false
   )
   &SolveForcesAndVelocity(Clamp Velocity=false, Limit Acceleration=false)
   &VC_UpdateRibbon_v1(Enable Skeletal Position=true)
   &ScaleColor(
    Scale Alpha=&FloatFromCurve(FloatCurve=DataInterface() {propertyValues {Curve {
        defaultValue=1
        keys [
         {interpMode=RCIM_Cubic, tangentMode=RCTM_User, time=0, value=0}
         {interpMode=RCIM_Cubic, tangentMode=RCTM_User, time=0.25, value=1}
         {tangentMode=RCTM_User, time=1, value=0}
        ]
       }}})
    Scale RGB=&Multiply_VectorByFloat(
     Vector=<1 1 1>
     Float=&FloatFromCurve(FloatCurve=DataInterface() {propertyValues {Curve {keys [<at 0 1>,<at 1 0.5>]}}})
    )
    ScaleA=true
    ScaleRGB=true
   )
  ]
  rendererClasses [&NiagaraRibbonRendererProperties]
  renderers [{rendererClass=&NiagaraRibbonRendererProperties}]
 }
 NE_VC_Sprites {
  emitterUpdateScript [
   &EmitterState(Life Cycle Mode=::NewEnumerator1, Loop Behavior=::NewEnumerator1, Loop Duration Mode=::NewEnumerator1, UseLoopDelay=false)
   &SpawnRate(SpawnRate=30, Use Spawn Probability=false)
   &SpawnRate(SpawnGroup=1, SpawnRate=35, Use Spawn Probability=false) {module=SpawnRate001}
   &SpawnRate(SpawnGroup=2, SpawnRate=20, Use Spawn Probability=false) {module=SpawnRate002}
  ]
  particleSpawnScript [
   &VC_PaintedVariantsOverride(
    Black=<rgba 0.005 0.005 0.005 2>
    BurntSienna=<rgba 12 1.55 0.6 1>
    Cobalt=<rgba 1.17 0.6 12 1>
    Crimson=<rgba 12 0.6 0.6 1>
    DefaultColor=<rgba 2.3333 0 7 1>
    ForestGreen=<rgba 0.3 12 0.7875 1>
    Grey=<rgba 0.85 0.85 0.85 2>
    Lime=<rgba 1.6667 5 0 1>
    Orange=<rgba 3 0.4 0 2>
    Pink=<rgba 10 0 15 1>
    Purple=<rgba 0.1333 0 0.4 1>
    Saffron=<rgba 10 10 0 1>
    SkyBlue=<rgba 0.6 12 12 1>
    TitaniumWhite=<rgba 0.4238 0.4242 0.45 2>
   )
   &InitializeParticle(
    Color=$Particles.Color
    Lifetime Max=&MaskFloatBySpawnGroup(Masked Float=1.15, Passthrough Float=0.6)
    Lifetime Min:NiagaraFloat=&MaskFloatBySpawnGroup(Masked Float=0.45, Passthrough Float=0.3)
    Lifetime Mode=::NewEnumerator1
    Sprite Rotation Mode=::NewEnumerator2
    Sprite Size Mode=::NewEnumerator1
    Uniform Sprite Size Max=&MaskFloatBySpawnGroup(Masked Float=2, Passthrough Float=10, Spawn Group Mask=2)
    Uniform Sprite Size Min=&MaskFloatBySpawnGroup(Masked Float=0.85, Passthrough Float=5, Spawn Group Mask=2)
    UsePositionOffset=false
    Write Material Random=true
   )
   &PartitionParticles(Particles Per Partition=4, Partition Mode:ENiagara_PartitionMode=::NewEnumerator1)
   &SkeletalMeshLocation(
    Bone / Socket Index=&Add_Integer(A=$Particles.PartitionParticles.Partition, B=4)
    Bone / Socket Sampling Mode=::NewEnumerator5
    Interpolate Mesh Sampling=false
    MaskSpawnGroup=false
    Mesh Sampling Type=::NewEnumerator6
    Sampled Position Offset=&AddVectorToPosition(Position=$User.offset, Vector=<-65 0 -20>)
    Skeletal Mesh=$System.SkeletalMesh
   )
   &AddVelocity(Velocity=&MaskVectorBySpawnGroup(
     Masked Vector=&RandomRangeVector(Minimum=<-450 -10 -5>, Maximum=<-450 10 10>)
     Passthrough Vector=&RandomRangeVector(Minimum=<-300 -50 -45>, Maximum=<-100 50 100>)
    ))
   &ShapeLocation(
    MaskSpawnGroup=false
    Override Seed=false
    Sphere Radius=&MaskFloatBySpawnGroup(Masked Float=20, Passthrough Float=15)
    UseOwnerScale=true
   )
   /SetVariables(Particles.RenderVisTag:NiagaraInt32=&SpawnGroup()) {isSetParameters=true}
  ]
  particleUpdateScript [
   &FN_DisableTireFX(
    Mask Index Attribute:NiagaraInt32=$Particles.PartitionParticles.Partition
    Mask Index FL=2
    Mask Index FR=3
    Mask Index RL=0
    Mask Index RR=1
   )
   &ParticleState()
   &InheritSourceMovement(
    Applied Position Scale=&Multiply_VectorByFloat(
     Vector=<1 1 1>
     Float=&FloatFromCurve(FloatCurve=DataInterface() {propertyValues {Curve {keys [
          {interpMode=RCIM_Cubic, leaveTangent=-0.1214, tangentMode=RCTM_Break, time=0, value=1}
          {arriveTangent=0.0008, interpMode=RCIM_Cubic, tangentMode=RCTM_Break, time=1, value=0.95}
         ]}}})
    )
    ApplyForce=false
    ApplyPosition=true
    ApplyPositionFalloff=false
   )
   &ScaleSpriteSize(
    Uniform Curve Scale=2.5
    Uniform Curve Sprite Scale=DataInterface() {propertyValues {Curve {keys [
        {interpMode=RCIM_Cubic, tangentMode=RCTM_Break, time=0.45, value=1}
        {arriveTangent=-3.0414, interpMode=RCIM_Cubic, tangentMode=RCTM_Break, time=1, value=0.2}
       ]}}}
   )
   &GravityForce(Gravity=&RandomRangeVector(Minimum=<0 0 -10>, Maximum=<0 0 20>))
   &SolveForcesAndVelocity(Clamp Velocity=false, Limit Acceleration=false)
  ]
  rendererClasses [&NiagaraSpriteRendererProperties]
  renderers [
   {rendererClass=&NiagaraSpriteRendererProperties}
   {rendererClass=&NiagaraSpriteRendererProperties, rendererIndex=1}
   {rendererClass=&NiagaraSpriteRendererProperties, rendererIndex=2}
  ]
 }
 NE_VC_WheelWrap {
  emitterUpdateScript [
   &EmitterState(
    Inactive Response=::NewEnumerator1
    Life Cycle Mode=::NewEnumerator1
    Loop Behavior=::NewEnumerator1
    Loop Duration Mode=::NewEnumerator1
    UseLoopDelay=false
   )
   &SpawnBurst_Instantaneous(Spawn Count=2, Use Spawn Probability=false, UseLoopCountLimit=false)
   /SetVariables(Emitter.TextureDesaturation:NiagaraFloat=&MakeCustomFloatFromBool(Bool=$User.bIsPainted, True Float=1, False Float=0)) {isSetParameters=true}
  ]
  particleSpawnScript [
   &VC_PaintedVariantsOverride(
    Black=<rgba 0.075 0.075 0.075 20>
    BurntSienna=<rgba 1 0.5833 0.5 10>
    Cobalt=<rgba 0.4375 0.5031 1.75 7.5>
    Crimson=<rgba 1.25 0.3646 0.1875 6>
    DefaultColor=<rgba 1.125 0.5 2 60>
    ForestGreen=<rgba 0.3125 1.25 0.3125 20>
    Grey=<rgba 0.6562 0.6562 0.7 4>
    Lime=<rgba 1.5833 2 1 5>
    Orange=<rgba 1.2 0.7 0 10>
    Pink=<rgba 1.3625 0.675 1.5 10>
    Purple=<rgba 0.1667 0 0.5 10>
    Saffron=<rgba 1.5 1.5 0.225 4>
    SkyBlue=<rgba 0.75 3 3 4>
    TitaniumWhite=<rgba 2 2 2 5>
   )
   &InitializeParticle(
    Color Mode=::NewEnumerator0
    Lifetime=5
    Mesh Scale Mode=::NewEnumerator4
    Sprite Rotation Mode=::NewEnumerator2
    Sprite Size Mode=::NewEnumerator3
    UsePositionOffset=false
    Write Material Random=true
   )
   &PartitionParticles()
   &SkeletalMeshLocation(
    Bone / Socket Index=&Add_Integer(A=$Particles.PartitionParticles.Partition, B=4)
    Bone / Socket Sampling Mode=::NewEnumerator5
    Interpolate Mesh Sampling=false
    MaskSpawnGroup=false
    Mesh Sampling Type=::NewEnumerator6
    Skeletal Mesh=$System.SkeletalMesh
   )
   &InitialMeshOrientation(Mesh Orientation Mode=::NewEnumerator7, Rotation=<0 0 0.5>, Use Rotation Vector=true)
   &Color(Color=<rgba 0 1 0 1>) {enabled=false}
  ]
  particleUpdateScript [
   &FN_DisableTireFX(
    Mask Index Attribute:NiagaraInt32=$Particles.PartitionParticles.Partition
    Mask Index FL=2
    Mask Index FR=3
    Mask Index RL=0
    Mask Index RR=1
   )
   &VC_WheelWrapScale(Wheel Index Attribute:NiagaraInt32=&Add_Integer(A=$Particles.PartitionParticles.Partition, B=2))
   &ParticleState(Kill Particles When Lifetime Has Elapsed=false, Loop Particles Lifetime=true)
   &UpdateMeshOrientation(Rotation Coordinate Space=::NewEnumerator2, Rotation Vector=<0 1 0>) {enabled=false}
   &SkeletalMeshLocation(
    Bone / Socket Index=&Add_Integer(A=$Particles.PartitionParticles.Partition, B=4)
    Bone / Socket Sampling Mode=::NewEnumerator5
    MaskSpawnGroup=false
    Mesh Sampling Type=::NewEnumerator6
    Skeletal Mesh=$System.SkeletalMesh
   ) {module=SkeletalMeshLocation001}
   &SolveForcesAndVelocity(Clamp Velocity=false, Limit Acceleration=false)
   &ScaleColor(ScaleA=true, ScaleRGB=true)
  ]
  rendererClasses [&NiagaraMeshRendererProperties]
  renderers [{rendererClass=&NiagaraMeshRendererProperties, rendererIndex=0}]
 }
}
summary {
 systemName=NS_DriftTrail_Singularity
 emitters [
  {emitterName=NE_VC_Sprites, rendererClasses [&NiagaraSpriteRendererProperties]}
  {emitterName=NE_VC_WheelWrap, rendererClasses [&NiagaraMeshRendererProperties]}
  {emitterName=NE_VC_Ribbon001, rendererClasses [&NiagaraRibbonRendererProperties]}
 ]
 userVariables [
  User.bIsFrontendPreview {type=NiagaraBool}
  User.bIsPainted {type=NiagaraBool}
  User.ColorVariantID {type=NiagaraInt32}
  User.offset {type=NiagaraPosition, description="", defaultValue {}}
  User.PaintedColor {type=LinearColor, description="", defaultValue {struct=LinearColor, value=<rgba 1 1 1 1>}}
  User.PaintedScalar {type=NiagaraFloat, defaultValue=1, description=""}
  User.Shine {type=NiagaraFloat, defaultValue=5, description=""}
  User.SpecMult {type=NiagaraFloat, defaultValue=0, description=""}
  User.SpecPower {type=NiagaraFloat, defaultValue=0, description=""}
  User.VehicleFlags {type=NiagaraInt32, defaultValue=806289408}
  User.Wheel_Radius_Array {
   type=NiagaraDataInterfaceArrayFloat
   defaultValue {
    struct=NiagaraExt_VariableValue_DataInterface
    value {
     dataInterface=&self:NiagaraDataInterfaceArrayFloat_1
     dataInterfaceClass=&NiagaraDataInterfaceArrayFloat
    }
   }
  }
  User.Wheel_Width_Array {
   type=NiagaraDataInterfaceArrayFloat
   defaultValue {
    struct=NiagaraExt_VariableValue_DataInterface
    value {
     dataInterface=&self:NiagaraDataInterfaceArrayFloat_0
     dataInterfaceClass=&NiagaraDataInterfaceArrayFloat
    }
   }
  }
  User.WheelTouchingArray {
   type=NiagaraDataInterfaceArrayFloat
   defaultValue {
    struct=NiagaraExt_VariableValue_DataInterface
    value {
     dataInterface=&self:NiagaraDataInterfaceArrayFloat_2
     dataInterfaceClass=&NiagaraDataInterfaceArrayFloat
    }
   }
  }
 ]
}
